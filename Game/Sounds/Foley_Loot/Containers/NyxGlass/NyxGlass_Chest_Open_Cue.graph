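# Sound cue for the opening of a NyxGlass chest, part of the loot container foley.
# A single wave player feeds a modulator that randomizes pitch while holding volume fixed.
# It uses the shared item-interact attenuation and the loot reward sound class, with a slight volume boost.
# A concurrency override prevents new instances from starting once the limit is reached, rather than stopping playing ones.
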SoundCue /Game/Sounds/Foley_Loot/Containers/NyxGlass/NyxGlass_Chest_Open_Cue
root=Modulator_0
nodes [
 {id=Modulator_0, children [WavePlayer_0], properties {pitchMax=1.1, volumeMax=1, volumeMin=1}}
 {id=WavePlayer_0, properties {soundWaveAssetPtr=&/NyxGlass_Chest_Open_01}}
]
properties {
 attenuationSettings=/Game/Sounds/Attenuation/Fort_Item_Interact_Att
 bOverrideConcurrency=true
 maxDistance=1850
 soundClassObject=&Fort_Loot_Reward_PSM
 volumeMultiplier=1.2
 concurrencyOverrides {resolutionRule=PreventNew}
}
